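# Masked master material for the YinYangCats cosmetic vehicle wheel, built on MF_WheelParent_Simple and enabled for skeletal meshes.
# A mask from vertex color blue times the inverted rim-mask green channel blends roughness and metallic toward Paw_PBR values.
# Those values are also multiplied by the mask, so partial coverage scales them down.
# The adjusted attributes then pass through MF_BaseVehicle_Damage and MF_VehicleDeRez_Transition.
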Material /VehicleCosmetics/Wheels/YinYangCats/Materials/M_Wheel_YinYangCats
graph {
 properties {
  bAutomaticallySetUsageInEditor=false
  bUseMaterialAttributes=true
  bUsedWithMeshDeformer=true
  bUsedWithSkeletalMesh=true
  blendMode=BLEND_Masked
  pixelDepthOffsetMode=PDOM_Legacy
  usedShadingModels=Default Lit
 }
 nodes [
  n1 = MaterialFunctionCall() {materialFunction=&MF_WheelParent_Simple}
  n2 = GetMaterialAttributes(None=@n1.MaterialAttributes) {attributeGetTypes [#1, #2]}
  n3 = Multiply(A=OneMinus(None=ComponentMask(None=@n1.RimRGBAMask) {g=true}), B=VertexColor().B)
  n4 = VectorParameter() {
   group=6_Additional Properties
   parameterName=Paw_PBR
   sortPriority=0
   channelNames {a=Unuse, b=Unuse, g=Metallic, r=Roughness}
   defaultValue=<rgba 0 0 0 1>
  }
  n5 = LinearInterpolate(A=@n2.Roughness, B=Multiply(A=@n4.Roughness, B=@n3), Alpha=@n3)
  n6 = LinearInterpolate(A=@n2.Metallic, B=Multiply(A=@n4.Metallic, B=@n3), Alpha=@n3)
  MaterialFunctionCall(In=MaterialFunctionCall(Material=SetMaterialAttributes!(MaterialAttributes=@n2.MaterialAttributes, Roughness=@n5, Metallic=@n6) {attributeSetTypes [#1, #2]}) {materialFunction=&MF_BaseVehicle_Damage}.Result) {materialFunction=&MF_VehicleDeRez_Transition}
  Reroute(None=@n1.VertexColorR)
 ]
}
